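; Gameplay ability for the Nitro Queen medallion that grants infinite ammo, except for ignore-listed weapons.
; On activation it caches pawn and controller, checks the current weapon, and binds the same check to weapon-equipped events.
; The check compares the weapon data's asset path against a DataTable ignore list, then sets the controller's infinite ammo flag.
; Ending the ability, as when the medallion is dropped, turns infinite ammo off.
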(event Ability|EventActivateAbility
  (|CallParentFunction)
  (CallFunction|SetReferences)
  (CallFunction|CheckValidWeapon (Weapon|GetCurrentWeapon (Variables|Default|GetPlayerPawn)))
  (FortPawn|BindEventtoOnWeaponEquippedDelegate (Variables|Default|GetPlayerPawn) (delegate Custom|CheckValidWeapon)))

(event Custom|CheckValidWeapon
  (params (NewWeapon :type Fort%20Weapon%20Object%20Reference) (PrevWeapon :type Fort%20Weapon%20Object%20Reference))
  (bind (_datatable _rowname) (Utilities|Struct|BreakDataTableRowHandle (Variables|Default|GetWeaponIgnoreList)))
  (bind _returnvalue_pathstring (Utilities|BreakSoftObjectPath (Utilities|GetSoftObjectPath (Data|GetWeaponData NewWeapon))))
  ;[1 0.1197 0.7025] Whenever the player switches weapons, we check if that weapon is in the Ignore List. If it is, we turn Off infinite ammo, if it's not, we turn infinite ammo On
  (bind _returnvalue (Utilities|GetDataTableRow "/MoonflaxMedallions/DataTables/MoonflaxMedallions_WeaponIgnoreList.MoonflaxMedallions_WeaponIgnoreList" _rowname))
  (bind _returnvalue_weaponignorelist_5_ed45afd048706f14db63e6b58c3370eb (Utilities|Struct|BreakBPSMoonflaxMedallionsIgnoreList _returnvalue))
  (bind _array_element (Utilities|Array|ForEachLoopwithBreak _returnvalue_weaponignorelist_5_ed45afd048706f14db63e6b58c3370eb)
    (:LoopBody
      (if (== _array_element (Utilities|String|StringToName _returnvalue_pathstring))
        (Variables|Default|SetIsWeaponValid false))
      (Variables|Default|SetIsWeaponValid true))
    (:Completed
      _))
  (CallFunction|UpdateInfiniteAmmo (Variables|Default|GetIsWeaponValid)))

(event Ability|EventOnEndAbility
  (params (bWasCancelled :type Boolean))
  ;[1 1 1] When players drop the Medallion, we turn infinite ammo Off
  (|CallParentFunction bWasCancelled)
  (Class|FortPlayerControllerAthena|SetHasInfiniteAmmo (Variables|Default|GetPlayerController) false))

(fn SetReferences
  (Variables|Default|SetPlayerPawn (CallFunction|GetAvatarasFortPlayerPawn))
  (bind _asfort_player_controller_athena (Utilities|Casting|CastToFortPlayerControllerAthena (Ability|GetPlayerControllerfromActorInfo)))
  (Variables|Default|SetPlayerController _asfort_player_controller_athena))

(fn UpdateInfiniteAmmo
  (params (IsWeaponValid :type Boolean))
  (Class|FortPlayerControllerAthena|SetHasInfiniteAmmo (Variables|Default|GetPlayerController) IsWeaponValid))
